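# Niagara effect for the HomeRange glider that adds wing-style ribbon trails and shed feathers.
# FN_AnimTrail draws two socket-pair ribbons, tinted by time-of-day color and faded by system opacity and an integrator.
# Feather spawns tumbling meshes from two sockets through staggered bursts plus a steady rate, with drag, vortex force and User.Variant blending two color palettes.
# Front-end modifiers change velocity, acceleration and trail behavior between gameplay and menu display.
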NiagaraSystem /BRCosmetics/Gadgets/Assets/VinderTech_GliderChute/Glider_HomeRange/FX/NS_Glider_HomeRange
dependencies {
 usedDataInterfaces [&NiagaraDataInterfaceColorCurve, &NiagaraDataInterfaceCurve, &NiagaraDataInterfaceSocketReader]
 usedDynamicInputs [
  &FNTimeOfDayColor
  /Niagara/DynamicInputs/Math/Min_Float
  &ClampFloat
  &Subtract_Float
  &Boolean_NotOperation
 ]
 usedModules [
  &self:SystemSpawnScript.SystemScriptSource.SystemScriptGraph.NiagaraNodeAssignment_3.SetVariables_#1
  &SystemState
  &Timeline
  &self:SystemSpawnScript.SystemScriptSource.SystemScriptGraph.NiagaraNodeAssignment_5.SetVariables_#2
  &self:SystemSpawnScript.SystemScriptSource.SystemScriptGraph.NiagaraNodeAssignment_0.SetVariables_#3
  &self:SystemSpawnScript.SystemScriptSource.SystemScriptGraph.NiagaraNodeAssignment_1.SetVariables_#4
  &self:SystemSpawnScript.SystemScriptSource.SystemScriptGraph.NiagaraNodeAssignment_2.SetVariables_#5
  &FN_Integrator
 ]
}
emitters {
 FN_AnimTrail {
  emitterSpawnScript [&FN_RibbonIDOffsetUtility()]
  emitterUpdateScript [
   &EmitterState()
   &FN_SpawnAnimTrail(Number of Anim Trails=2)
  ]
  particleSpawnScript [
   &FN_InitializeAnimTrailBoneIndecies(First Bone or Socket Index=1, Second Bone or Socket Index=4)
   &FN_InitializeAnimTrailBoneIndecies(AnimTrail Selection=1, First Bone or Socket Index=3, Second Bone or Socket Index=5) {module=FN_InitializeAnimTrailBoneIndecies001}
   &FN_InitializeAnimTrail(AnimTrailWidth=1.15, Lifetime=0.152)
   &AddVelocity(Velocity=<0 400 0>) {enabled=false}
   &AddVelocity(Velocity=&FN_FrontEndVector3Modifier(Gameplay=<0 0 0>, Front End=<0 1200 0>)) {module=AddVelocity001}
  ]
  particleUpdateScript [
   &SolveForcesAndVelocity(Clamp Velocity=false, Limit Acceleration=false)
   &FN_UpdateAnimTrail() {module=FN_UpdateAnimTrail001}
   &InterpolateOverTime(Initial Float Value:NiagaraFloat=$System.Timeline001.Time, Target Float Value=1, Use Set Interpolation Rate=true)
   &KillParticles(Kill Particles=&SetBoolByFloatComparison(
     A=&FN_FrontEndFloatModifier(Gameplay=$System.Timeline001.Time, Front End=1)
     B=$Particles.InterpolateOverTime.InterpolateOverTime_Float.Moving Average
     Comparison Type=::NewEnumerator6
    ))
   &ScaleColor(
    Scale Alpha=&FloatFromCurve(
     FloatCurve=DataInterface() {propertyValues {Curve {keys [<at 0 0 RCIM_Cubic>, {arriveTangent=1, interpMode=RCIM_Cubic, leaveTangent=1, time=0.0755, value=1}, <at 1 1>]}}}
     Scale Curve=&Multiply_Float(
      A=&Multiply_Float(A=0.15, B=&FNTimeOfDayScalar(Night Scale=1, Morning Scale=1, Evening Scale=1))
      B=&Multiply_Float(
       A=$System.Opacity
       B=&FN_FrontEndFloatModifier(Gameplay=$System.FN_Integrator001.NormalizedValue, Front End=0.7)
      )
     )
    )
    Scale RGB=&MakeVectorFromLinearColorRGB(Color=$System.ToDColor)
    ScaleA=true
    ScaleRGB=true
   )
  ]
  rendererClasses [&NiagaraRibbonRendererProperties]
  renderers [
   {rendererClass=&NiagaraRibbonRendererProperties}
   {rendererClass=&NiagaraRibbonRendererProperties, rendererIndex=1}
  ]
 }
 Feather {
  emitterUpdateScript [
   &EmitterState(Life Cycle Mode=::NewEnumerator1, Loop Behavior=::NewEnumerator1, Loop Duration Mode=::NewEnumerator1, UseLoopDelay=false)
   &SpawnRate(SpawnRate=3, Use Spawn Probability=false)
   &SpawnBurst_Instantaneous(
    Spawn Count=&RandomRangeInt(Minimum=4, Maximum=6)
    Use Spawn Probability=false
    UseLoopCountLimit=false
   )
   &SpawnBurst_Instantaneous(
    Spawn Count=&RandomRangeInt(Minimum=1, Maximum=2)
    Spawn Time=0.15
    Use Spawn Probability=false
    UseLoopCountLimit=false
   ) {module=SpawnBurst_Instantaneous001}
   &SpawnBurst_Instantaneous(
    Spawn Count=&RandomRangeInt(Minimum=1, Maximum=2)
    Spawn Time=0.3
    Use Spawn Probability=false
    UseLoopCountLimit=false
   ) {module=SpawnBurst_Instantaneous002}
  ]
  particleSpawnScript [
   &FN_SetSpawnGroup(Number of Spawn Groups=2)
   &InitializeParticle(
    Color=&LerpLinearColors(
     StartColor=&ColorFromCurve(
      ColorCurve=DataInterface() {propertyValues {
        AlphaCurve {keys [<at 0 1>]}
        BlueCurve {keys [<at 0.2258 0.0938 RCIM_Cubic>,<at 0.4907 0.026>,<at 0.8108 0.1302 RCIM_Cubic>]}
        GreenCurve {keys [<at 0.2258 0.0417 RCIM_Cubic>,<at 0.4907 0.026>,<at 0.8108 0.0668 RCIM_Cubic>]}
        RedCurve {keys [<at 0.2258 0.0628 RCIM_Cubic>,<at 0.4907 0.026>,<at 0.8108 0.0258 RCIM_Cubic>]}
       }}
      CurveIndex=&RandomRangeFloat(Minimum=0, Maximum=1)
     )
     EndColor=&ColorFromCurve(
      ColorCurve=DataInterface() {propertyValues {
        AlphaCurve {keys [<at 0 1 RCIM_Cubic>]}
        BlueCurve {keys [<at 0 0.01 RCIM_Cubic>,<at 1 0.016 RCIM_Cubic>]}
        GreenCurve {keys [<at 0 0.01 RCIM_Cubic>,<at 1 0.016 RCIM_Cubic>]}
        RedCurve {keys [<at 0 0.01 RCIM_Cubic>,<at 1 0.016 RCIM_Cubic>]}
       }}
      CurveIndex=&RandomRangeFloat(Minimum=0, Maximum=1)
     )
     LerpFactor=$User.Variant
    )
    Lifetime=&RandomRangeFloat(Minimum=1, Maximum=1.5)
    Mesh Scale=&Multiply_Vector(
     A=$Engine.Owner.Scale
     B=/Niagara/DynamicInputs/CascadeConversion/SplitVector(
      Double Component Vector Channels:ECascadeNiagaraTwoVectorChannels=::NewEnumerator0
      Double Component:NiagaraFloat=&RandomRangeFloat(Minimum=0.3, Maximum=0.5)
      Single Component:NiagaraFloat=&RandomRangeFloat(Minimum=0, Maximum=1)
     )
    )
    Mesh Scale Mode=::NewEnumerator4
    UsePositionOffset=false
    Write Material Random=true
   )
   &AddVelocity(
    Rotation Coordinate Space=::Simulation
    Velocity=&FN_FrontEndVector3Modifier(
     Gameplay=&RandomRangeVector(Minimum=<-150 -150 -50>, Maximum=<150 150 -800>)
     Front End=&RandomRangeVector(Minimum=<-150 -150 -222>, Maximum=<150 150 222>)
    )
   )
   &AddVelocity(
    Rotation Coordinate Space=::Simulation
    Velocity=&Multiply_VectorByFloat(Vector=$Particles.ShapeLocation.ShapeNormal, Float=50)
   ) {module=AddVelocity001}
   &DynamicMaterialParameters(
    Index 0 Param 1=&RandomRangeFloat(Minimum=0, Maximum=0.9)
    Index 0 Param 2=9
    Index 0 Param 4=&RandomRangeFloat(Minimum=0, Maximum=1)
    Module.Param0WriteEnabled=true
    Module.Param1WriteEnabled=true
    Module.Param2WriteEnabled=true
    Module.Param3WriteEnabled=true
   )
   &InitialMeshOrientation(Use Rotation Vector=false)
   /SetVariables(Particles.SpawnRotationVector=&RandomRangeVector(Minimum=<0 0 0>, Maximum=<1 1 1>)) {isSetParameters=true}
   &SocketLocation(
    SamplePosOffset=true
    Sampled Position Offset=&MergeHelper_VectorSwitch(
     0=&RandomRangeVector(Minimum=<10 -20 0>, Maximum=<-30 -15 1>)
     1=&RandomRangeVector(Minimum=<-10 15 0>, Maximum=<30 20 1>)
     2=?
     3=?
     Index=&SpawnGroup()
     Num=::NewEnumerator0
    )
    Socket Index=&MergeHelper_IntSwitch(0=0, 1=3, 2=?, 3=?, 4=?, 5=?, Index=&SpawnGroup(), Num=::NewEnumerator0)
    Socket Reader=$System.Socket Reader
    Socket Sampling Mode=::NewEnumerator2
    VelocitySpeedLimit=false
   )
  ]
  particleUpdateScript [
   &ParticleState()
   &UpdateMeshOrientation(
    Rotation Rate=&FloatFromCurve(FloatCurve=DataInterface() {propertyValues {Curve {keys [<at 0 1>,<at 1 0>]}}}, Scale Curve=2)
    Rotation Vector=$Particles.SpawnRotationVector
   )
   &ScaleMeshSize(Scale Factor=&VectorFromFloat(Value=&FloatFromCurve(FloatCurve=DataInterface() {propertyValues {Curve {keys [<at 0 0 RCIM_Cubic>, <at 0.1 1>, {arriveTangent=-1.1111, interpMode=RCIM_Cubic, leaveTangent=-1.1111, time=0.5, value=1}, <at 1 0>]}}})))
   &Drag(Drag=2, Use Linear Drag=true, Use Rotational Drag=false)
   &VortexForce(
    UseOriginOffset=true
    Vortex Axis=<1 1 1>
    Vortex Axis Coordinate Space:ENiagaraCoordinateSpace=::Local
    Vortex Force Amount=30
    Vortex Origin Offset:Vector3f=&RandomRangeVector(Minimum=<-70 -50 0>, Maximum=<70 -70 0>)
    Vortex Origin Offset Coordinate Space:ENiagaraCoordinateSpace=::Local
   )
   &InheritSourceMovement(Applied Position Scale=<0.75 0.75 0.25>, ApplyForce=false, ApplyPosition=true, ApplyPositionFalloff=false)
   &AccelerationForce(
    Acceleration=&FN_FrontEndVector3Modifier(Gameplay=<0 0 0>, Front End=&RandomRangeVector(Minimum=<-666 666 0>, Maximum=<666 3333 666>))
    Coordinate Space=::Local
   )
   &SolveForcesAndVelocity(Clamp Velocity=false, Limit Acceleration=false)
  ]
  rendererClasses [&NiagaraMeshRendererProperties]
  renderers [{rendererClass=&NiagaraMeshRendererProperties, rendererIndex=0}]
 }
}
summary {
 systemName=NS_Glider_HomeRange
 emitters [
  {emitterName=Feather, rendererClasses [&NiagaraMeshRendererProperties]}
  {emitterName=FN_AnimTrail, rendererClasses [&NiagaraRibbonRendererProperties]}
 ]
 userVariables [
  User.bIsBigFlapping {type=NiagaraBool}
  User.bIsFrontEnd {type=NiagaraBool}
  User.bIsFrontEndPreview {type=NiagaraBool}
  User.bIsFullyDeployed {type=NiagaraBool}
  User.blsFlapping {type=NiagaraBool}
  User.FlapCurve {type=NiagaraFloat, defaultValue=0, description=""}
  User.ForwardDot {type=NiagaraFloat, defaultValue=1, description=""}
  User.RibbonIDOffset {type=NiagaraInt32}
  User.RightDot {type=NiagaraFloat, defaultValue=0, description=""}
  User.Variant {type=NiagaraFloat, defaultValue=0, description=""}
 ]
}
